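# Winter-art material layer function that builds a PBR layer from diffuse, spec and normal textures and height-blends it over a background material.
# Textures share world-XY planar UVs with scale and offset inputs; the spec map's green channel drives metallic, specular and roughness alike.
# Normals are rebuilt for world space, optionally converted to tangent space, then XY-boosted.
# Diffuse alpha becomes Opacity height for HeightLerp; both layer and blend are output.
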MaterialFunction /Game/Athena/WinterArt/Materials/MF_WS_Layer
graph {
 properties {
  bExposeToLibrary=true
  description=PBR Material Layer. Inputs for textures - output complit material set with height map in Opacity. Default normal in World Space
  libraryCategoriesText [Texturing]
 }
 nodes [
  n1 = FunctionInput() {bUsePreviewValueAsDefault=true, inputName=Background Material, inputType=FunctionInput_MaterialAttributes, sortPriority=11}
  n2 = WorldPosition()
  n3 = FunctionInput() {bUsePreviewValueAsDefault=true, inputName=ScaleLayer, inputType=FunctionInput_Scalar, sortPriority=3, previewValue=<1024 0 0 1>}
  n4 = Abs(None=@n3)
  n5 = Add(
   A=Divide(A=ComponentMask(None=@n2.XYZ) {g=true, r=true}, B=Multiply(A=@n4) {constB=-1})
   B=FunctionInput() {bUsePreviewValueAsDefault=true, inputName=UV_Distiortion, inputType=FunctionInput_Scalar, sortPriority=3}
  )
  n6 = FunctionInput() {bUsePreviewValueAsDefault=true, inputName=Diffuse, inputType=FunctionInput_Texture2D, previewValue=<1 0 0 1>}
  n7 = TextureSample(UVs=@n5, Tex=@n6)
  n8 = FunctionInput() {bUsePreviewValueAsDefault=true, inputName=Desaturation, inputType=FunctionInput_Scalar, sortPriority=1, previewValue=<0 1 1 1>}
  n9 = FunctionInput() {bUsePreviewValueAsDefault=true, inputName=Spec, inputType=FunctionInput_Texture2D, sortPriority=1, previewValue=<1 0 0 1>}
  n10 = FunctionInput() {bUsePreviewValueAsDefault=true, inputName=SpecMask_Tweak, sortPriority=3, previewValue=<1 1 1 1>}
  n11 = MaterialFunctionCall(Float3=Multiply(A=TextureSample(UVs=@n5, Tex=@n9) {samplerType=SAMPLERTYPE_Masks}.RGB, B=@n10)) {materialFunction=&BreakOutFloat3Components}
  n12 = FunctionInput() {bUsePreviewValueAsDefault=true, inputName=Normal, inputType=FunctionInput_Texture2D, sortPriority=2, previewValue=<0 0 1 1>}
  n13 = TextureSample(UVs=@n5, Tex=@n12) {samplerType=SAMPLERTYPE_Normal}
  n14 = Multiply(A=@n13.RGB, B=Constant3Vector() {constant=<rgba -1 -1 1 0>})
  n15 = Constant3Vector() {constant=<rgba 0 0 1 0>}
  n16 = ComponentMask(None=Multiply(A=Transform(None=@n15))) {b=true}
  n17 = Normalize(VectorInput=AppendVector(A=ComponentMask(None=@n14) {g=true, r=true}, B=Multiply(A=@n13.B, B=@n16)))
  n18 = Transform(None=@n17) {transformSourceType=TRANSFORMSOURCE_World, transformType=TRANSFORM_Tangent}
  n19 = FunctionInput(Preview=StaticBool()) {bUsePreviewValueAsDefault=true, inputName=Tangent Space Normal, inputType=FunctionInput_StaticBool, sortPriority=4}
  n20 = FunctionInput() {bUsePreviewValueAsDefault=true, inputName=BoostNormals, inputType=FunctionInput_Scalar, previewValue=<1.5 0 0 1>}
  n21 = AppendVector(A=@n20, B=@n20)
  n22 = Constant() {r=1}
  n23 = SetMaterialAttributes!(
   MaterialAttributes
   BaseColor=Multiply(
    A=Desaturation(None=@n7.RGB, Fraction=@n8)
    B=FunctionInput() {bUsePreviewValueAsDefault=true, inputName=Mod Color, sortPriority=1, previewValue=<1 1 1 1>}
   )
   Metallic=@n11.G
   Specular=@n11.G
   Roughness=@n11.G
   Normal=Multiply(A=StaticSwitch(True=@n18, False=@n17, Value=@n19), B=AppendVector(A=@n21, B=@n22))
   Opacity=@n7.A
  ) {attributeSetTypes [
    #1
    #2
    #3
    #4
    #5
    #6
   ]}
  n24 = FunctionInput() {bUsePreviewValueAsDefault=true, inputName=Transition Phase, inputType=FunctionInput_Scalar, sortPriority=5, previewValue=<0.5 0 0 1>}
  n25 = FunctionInput() {bUsePreviewValueAsDefault=true, inputName=HM Boost, inputType=FunctionInput_Scalar, sortPriority=6, previewValue=<1 0 0 1>}
  n26 = Multiply(A=GetMaterialAttributes(None=@n23) {attributeGetTypes [#6]}.Opacity, B=@n25)
  n27 = FunctionInput() {bUsePreviewValueAsDefault=true, inputName=Contrast, inputType=FunctionInput_Scalar, sortPriority=7, previewValue=<0.5 0 0 1>}
  FunctionOutput(None=BlendMaterialAttributes(
    A=@n1
    B=FunctionOutput(None=@n23) {description=PBR material + Height in Opacity, outputName=OuputLayer}
    Alpha=MaterialFunctionCall(Transition Phase=@n24, Height Texture=@n26, Contrast=@n27) {materialFunction=&HeightLerp}.Alpha
   )) {description=A B Blend, outputName=Blended Layers}
 ]
}
