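# Garbage island material for both the ground surface and scattered grass, seaweed or trash meshes, chosen by static switches.
# MF_WS_Layer stacks dirt, trash, seaweed, puddles, more trash and grass using vertex colors, noise and height lerps.
# World height against an MPC or parameter water level drives wetness, underwater tint and, on the ground, downward, normal-widened vertex offset.
# Scatter variants use dithered depth offset and instance-faded masks; Forward/Mobile keep only base color, also wired into normal.
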Material /Game/Athena/Apollo/Environments/Meshes/Blockout/S13/GarbageIsland/Materials/M_GarbageFloat
graph {
 properties {
  bTangentSpaceNormal=false
  bUseMaterialAttributes=true
  bUsedWithInstancedStaticMeshes=true
  bUsedWithNanite=true
  pixelDepthOffsetMode=PDOM_Legacy
  usedShadingModels=Default Lit
 }
 nodes [
  n1 = TextureObjectParameter() {
   group=Set3
   parameterName=Set3 Mask
   samplerType=SAMPLERTYPE_Masks
   texture=/Game/Environments/Apollo/Sets/Terrain/Textures/T_Apollo_Forest_Floor_S_01
  }
  n2 = TextureObjectParameter() {
   group=Set3
   parameterName=Set3 Normal
   samplerType=SAMPLERTYPE_Normal
   texture=/Game/Environments/Apollo/Sets/Terrain/Textures/T_Apollo_Forest_Floor_N_01
  }
  n3 = TextureObjectParameter() {group=Set4, parameterName=Set4 Diffuse, texture=/Game/Environments/Apollo/Sets/GarbageIsland/Textures/T_GarbageLayer2_TrashDebris_D}
  n4 = TextureObjectParameter() {
   group=Set4
   parameterName=Set4 Mask
   samplerType=SAMPLERTYPE_Masks
   texture=/Game/Environments/Apollo/Sets/GarbageIsland/Textures/T_GarbageLayer2_TrashDebris_S
  }
  n5 = VectorParameter() {group=Set4, parameterName=Set4, defaultValue=<rgba 0.5 0.4814 0.4115 0>}
  n6 = ComponentMask(None=WorldPosition().XYZ) {g=true, r=true}
  n7 = TextureSample(UVs=Divide(A=@n6, B=ScalarParameter() {defaultValue=5555, parameterName=ColorNoiseSize})) {texture=/Game/Environments/Apollo/Rocks/Natural_Rocks/Textures/T_RockDetails_02Db}
  n8 = CollectionParameter() {collection=/Game/Athena/Apollo/Environments/Landscape/MPC/Apollo_MPC_Terrain, parameterName=WaterLevel}
  n9 = ScalarParameter() {defaultValue=323, group=WorldWater, parameterName=WaterHeight}
  n10 = Subtract(
   A=ComponentMask(None=WorldPosition().XYZ) {b=true}
   B=MaterialProxyReplace(Realtime=StaticSwitchParameter(True=@n8, False=@n9) {group=WorldWater, parameterName=UseMPC}, MaterialProxy=Constant())
  )
  n11 = Saturate(None=Divide(A=@n10, B=Constant() {r=895.2903}))
  n12 = TextureObjectParameter() {
   group=Set4
   parameterName=Set4 Normal
   samplerType=SAMPLERTYPE_Normal
   texture=/Game/Environments/Apollo/Sets/GarbageIsland/Textures/T_GarbageLayer2_TrashDebris_N
  }
  n13 = TextureObjectParameter() {
   group=Seaweed
   parameterName=Seaweed Mask
   samplerType=SAMPLERTYPE_Masks
   texture=/Game/Environments/Apollo/Sets/GarbageIsland/Textures/T_SeaweedLayer1_S
  }
  n14 = TextureSample(UVs=Divide(A=@n6, B=Constant() {r=3200})) {samplerSource=SSM_Wrap_WorldGroupSettings, samplerType=SAMPLERTYPE_Grayscale, texture=&LowResBlurredNoise}
  n15 = TextureObjectParameter() {
   group=Seaweed
   parameterName=Seaweed Normal
   samplerType=SAMPLERTYPE_Normal
   texture=/Game/Environments/Apollo/Sets/GarbageIsland/Textures/T_SeaweedLayer1_N
  }
  n16 = MaterialFunctionCall!(
   Diffuse=TextureObjectParameter() {group=Seaweed, parameterName=Seaweed Diffuse, texture=/Game/Environments/Apollo/Sets/GarbageIsland/Textures/T_SeaweedLayer1_D}
   BoostNormals
   Spec=@n13
   Mod Color=LinearInterpolate(
    A=VectorParameter() {group=Seaweed, parameterName=Seaweed Color1, defaultValue=<rgba 1.2 0.5375 0.5375 0>}.RGB
    B=VectorParameter() {group=Seaweed, parameterName=Seaweed Color2, defaultValue=<rgba 0.8502 1.0906 1.2 0>}.RGB
    Alpha=@n14.R
   )
   Desaturation
   Normal=@n15
   ScaleLayer=ScalarParameter() {defaultValue=820, group=Seaweed, parameterName=Seaweed_Scale}
   SpecMask_Tweak
   UV_Distiortion
   Tangent Space Normal
   Transition Phase
   HM Boost
   Contrast
   Background Material
  ) {materialFunction=/Game/Athena/WinterArt/Materials/MF_WS_Layer}
  n17 = Saturate(None=Divide(A=@n10, B=ScalarParameter() {defaultValue=43.7118, group=Seaweed, parameterName=Seaweeds Fade}))
  n18 = TextureObjectParameter() {group=Set4, parameterName=Set4 Diffuse, texture=&T_GarbageLayer2_TrashDebris_D}
  n19 = TextureObjectParameter() {
   group=Set4
   parameterName=Set4 Mask
   samplerType=SAMPLERTYPE_Masks
   texture=&T_GarbageLayer2_TrashDebris_S
  }
  n20 = TextureObjectParameter() {
   group=Set4
   parameterName=Set4 Normal
   samplerType=SAMPLERTYPE_Normal
   texture=&T_GarbageLayer2_TrashDebris_N
  }
  n21 = Constant() {r=0.45}
  n22 = VertexColor()
  n23 = TextureObjectParameter() {
   group=Set2
   parameterName=Set2 Mask
   samplerType=SAMPLERTYPE_Masks
   texture=/Game/Environments/Apollo/Sets/Terrain/Textures/T_Apollo_Dirt_S_01
  }
  n24 = TextureObjectParameter() {
   group=Set2
   parameterName=Set2 Normal
   samplerType=SAMPLERTYPE_Normal
   texture=/Game/Environments/Apollo/Sets/Terrain/Textures/T_Apollo_Dirt2_N
  }
  n25 = GetMaterialAttributes(None=MaterialFunctionCall(
    Diffuse=TextureObjectParameter() {group=Set2, parameterName=Set2 Diffuse, texture=/Game/Environments/Apollo/Sets/Terrain/Textures/T_Apollo_Dirt2_D}
    Spec=@n23
    Mod Color=@n7.RGB
    Desaturation=Constant() {r=1}
    Normal=@n24
    ScaleLayer=ScalarParameter() {defaultValue=512, group=Set2, parameterName=Set2_Scale}
   ) {materialFunction=&MF_WS_Layer}.OuputLayer) {attributeGetTypes [#1]}
  n26 = LinearInterpolate(A=@n25.Normal, B=Transform(None=Constant3Vector() {constant=<rgba 0 0 1 0>})) {constAlpha=0.75}
  n27 = LinearInterpolate(A=@n26, B=@n25.Normal, Alpha=@n14.R)
  n28 = MaterialFunctionCall(
   Diffuse=@n18
   Spec=@n19
   Mod Color=@n7.RGB
   Desaturation=ScalarParameter() {defaultValue=0.7, group=Set4, parameterName=GroundTrashDesaturation}
   Normal=@n20
   ScaleLayer=ScalarParameter() {defaultValue=1024, group=Set4, parameterName=Set 4 Scale}
   Transition Phase=Subtract(A=Max(A=@n14.R, B=@n21), B=Multiply(A=@n22.A) {constB=0.15})
   Background Material=SetMaterialAttributes!(MaterialAttributes=@n25.MaterialAttributes, Normal=LinearInterpolate(A=@n27, B=@n26, Alpha=@n22.A)) {attributeSetTypes [#1]}
  ) {materialFunction=&MF_WS_Layer}
  n29 = Multiply(
   A=Multiply(
    A=GetMaterialAttributes(None=@n28.OuputLayer) {attributeGetTypes [#2]}.Opacity
    B=ScalarParameter() {defaultValue=2, group=Seaweed, parameterName=Seaweed HM Transition}
   )
   B=Multiply(A=@n14.R, B=ScalarParameter() {defaultValue=5, group=Seaweed, parameterName=Seaweeds Noise})
  )
  n30 = GetMaterialAttributes(None=@n28.Blended Layers) {attributeGetTypes [#3]}
  n31 = Constant() {r=1.5}
  n32 = LinearInterpolate(B=@n31, Alpha=@n22.A) {constA=1}
  n33 = LinearInterpolate(
   A=LinearInterpolate(A=Multiply(A=@n30.Base Color, B=@n32), B=@n30.Base Color, Alpha=@n14.R)
   B=Multiply(A=@n30.Base Color, B=@n31)
   Alpha=@n14.R
  )
  n34 = StaticSwitchParameter(
   True=@n16.OuputLayer
   False=BlendMaterialAttributes(
    A=@n16.OuputLayer
    B=SetMaterialAttributes!(MaterialAttributes=@n30.MaterialAttributes, Base Color=@n33) {attributeSetTypes [#3]}
    Alpha=MaterialFunctionCall(Transition Phase=@n17, Height Texture=@n29) {materialFunction=&HeightLerp}.Lerp Alpha No Contrast
   )
  ) {group=Scatter, parameterName=Seaweeds}
  n35 = GetMaterialAttributes(None=@n34) {attributeGetTypes [#3, #4, #5]}
  n36 = ScalarParameter() {defaultValue=75, group=WorldWater, parameterName=SoftWaterline Blur}
  n37 = BlendMaterialAttributes(
   A=SetMaterialAttributes!(
    MaterialAttributes=@n35.MaterialAttributes
    Base Color=Multiply(A=@n35.Base Color) {constB=0.5}
    Roughness=Multiply(A=@n35.Roughness) {constB=0.3}
    Specular=Multiply(A=@n35.Specular) {constB=0.5}
   ) {attributeSetTypes [#3, #4, #5]}
   B=@n34
   Alpha=Saturate(None=Divide(A=@n10, B=@n36))
  )
  n38 = Multiply(A=GetMaterialAttributes(None=@n37) {attributeGetTypes [#3, #1]}.Base Color) {constB=0.5}
  n39 = LinearInterpolate(
   A=LinearInterpolate(
    A=VectorParameter() {group=Puddle, parameterName=PuddleColor, defaultValue=<rgba 0.0703 0.0604 0.0434 0>}.RGB
    B=VectorParameter() {group=Puddle, parameterName=PuddleColorHigh, defaultValue=<rgba 0.1276 0.0857 0.0473 0>}.RGB
    Alpha=@n11
   )
   B=@n38
  ) {constAlpha=0.15}
  n40 = Fresnel(Normal=VertexNormalWS()) {exponent=1.5}
  n41 = Multiply(A=@n39, B=OneMinus(None=@n40))
  n42 = Subtract(A=VertexColor().B, B=Constant() {r=0.4034})
  n43 = GetMaterialAttributes(None=@n28.OuputLayer) {attributeGetTypes [#2]}
  n44 = Constant() {r=1}
  n45 = Constant3Vector() {constant=<rgba 0 0 1 0>}
  n46 = Divide(A=ComponentMask(None=WorldPosition().XYZ) {g=true, r=true}, B=Constant() {r=200})
  n47 = Constant() {r=0.05}
  n48 = TextureSample(UVs=Panner(Coordinate=@n46, Speed=@n47)) {
   samplerSource=SSM_Wrap_WorldGroupSettings
   samplerType=SAMPLERTYPE_Normal
   texture=/Game/Gadgets/Assets/Telephone_Booth/Textures/T_BlindsNormalNoise_N
  }
  n49 = TextureSample(UVs=Panner(Coordinate=@n46, Speed=Multiply(A=@n47) {constB=-2})) {
   samplerSource=SSM_Wrap_WorldGroupSettings
   samplerType=SAMPLERTYPE_Normal
   texture=&T_BlindsNormalNoise_N
  }
  n50 = Constant3Vector() {constant=<rgba 1 1 0 0>}
  n51 = LinearInterpolate(A=@n45, B=Add(A=@n48.RGB, B=Multiply(A=@n49.RGB, B=@n50)), Alpha=Constant() {r=0.05})
  n52 = SetMaterialAttributes!(
   MaterialAttributes
   Base Color=LinearInterpolate(
    A=MaterialFunctionCall(A=@n38, B=@n41, Transition Phase=@n42, Height Texture=@n43.Opacity, Contrast=@n44) {materialFunction=&HeightLerp}.Results
    B=@n39
    Alpha=@n40
   )
   Roughness=Constant() {r=0.025}
   Specular=Multiply(A=Constant() {r=0.5}, B=@n40)
   Normal=Transform(None=LinearInterpolate(A=@n45, B=@n51, Alpha=@n40))
  ) {attributeSetTypes [
    #3
    #4
    #5
    #1
   ]}
  n53 = VertexColor()
  n54 = ComponentMask(None=WorldPosition().XYZ) {g=true, r=true}
  n55 = Constant() {r=4000}
  n56 = TextureSample(UVs=Panner(Coordinate=Divide(A=@n54, B=@n55), Speed=Constant() {r=0.05})) {samplerSource=SSM_Wrap_WorldGroupSettings, samplerType=SAMPLERTYPE_Grayscale, texture=&LowResBlurredNoise}
  n57 = Subtract(
   A=OneMinus(None=GetMaterialAttributes(None=@n28.OuputLayer) {attributeGetTypes [#2]}.Opacity)
   B=Multiply(A=@n56.R, B=Constant() {r=0.8})
  )
  n58 = MaterialFunctionCall(
   Diffuse=@n3
   Spec=@n4
   Mod Color=LinearInterpolate(A=Multiply(A=@n5.RGB, B=@n7.RGB) {constB=0.5}, B=@n5.RGB, Alpha=@n11)
   Desaturation=LinearInterpolate(B=ScalarParameter() {defaultValue=0.5, group=Set4, parameterName=Set4 Desaturation}, Alpha=@n11) {constA=0.2}
   Normal=@n12
   ScaleLayer=ScalarParameter() {defaultValue=1024, group=Set4, parameterName=Set 4 Scale}
   SpecMask_Tweak=LinearInterpolate(A=Constant3Vector() {constant=<rgba 0.6523 1 0.9501 1>}, B=Constant3Vector() {constant=<rgba 1 1 1 0>}, Alpha=@n11)
   Transition Phase=VertexColor().R
   HM Boost=ScalarParameter() {defaultValue=1.5, group=Set4, parameterName=Set 4 Boost HM}
   Contrast=ScalarParameter() {defaultValue=0.1, group=Set4, parameterName=Set 4 Contrast}
   Background Material=BlendMaterialAttributes(
    A=@n37
    B=@n52
    Alpha=MaterialFunctionCall(Transition Phase=@n53.B, Height Texture=@n57) {materialFunction=&HeightLerp}.Lerp Alpha No Contrast
   )
  ) {materialFunction=&MF_WS_Layer}
  n59 = MaterialFunctionCall(
   Diffuse=TextureObjectParameter() {group=Set3, parameterName=Set3 Diffuse, texture=/Game/Environments/Apollo/Sets/Terrain/Textures/T_Apollo_Grass2_D}
   BoostNormals=Constant() {r=1}
   Spec=@n1
   Normal=@n2
   ScaleLayer=ScalarParameter() {defaultValue=512, group=Set3, parameterName=Set3_Scale}
   Transition Phase=VertexColor().G
   Background Material=@n58.Blended Layers
  ) {materialFunction=&MF_WS_Layer}
  n60 = GetMaterialAttributes(None=@n59.OuputLayer) {attributeGetTypes [#3, #2]}
  n61 = CollectionParameter() {collection=&Apollo_MPC_Terrain, parameterName=WaterLevel}
  n62 = ScalarParameter() {defaultValue=323, group=WorldWater, parameterName=WaterHeight}
  n63 = Subtract(
   A=ComponentMask(None=WorldPosition().XYZ) {b=true}
   B=MaterialProxyReplace(Realtime=StaticSwitchParameter(True=@n61, False=@n62) {group=WorldWater, parameterName=UseMPC}, MaterialProxy=Constant())
  )
  n64 = Constant() {r=1200}
  n65 = LinearInterpolate(A=VectorParameter() {parameterName=WetTone, defaultValue=<rgba 0.8359 0.7182 0.5355 1>}.RGB, Alpha=Saturate(None=Divide(A=@n63, B=@n64)))
  n66 = Multiply(A=@n65, B=@n60.Base Color)
  n67 = Multiply(A=Multiply(A=@n60.Opacity, B=ScalarParameter() {defaultValue=1, group=Scatter, parameterName=GrassMask}), B=PerInstanceFadeAmount())
  n68 = Multiply(A=Constant() {r=5}, B=MaterialFunctionCall() {materialFunction=&DitherTemporalAA}.Result)
  n69 = GetMaterialAttributes(None=@n37) {attributeGetTypes [#3, #2, #5]}
  n70 = VectorParameter() {parameterName=Underwater Color, defaultValue=<rgba 0.4765 0.7786 0.649 0>}
  n71 = Add(A=@n63, B=ScalarParameter() {defaultValue=151.6065, parameterName=Waterline Offset})
  n72 = ScalarParameter() {defaultValue=512, parameterName=Waterline Fade}
  n73 = GetMaterialAttributes(None=@n58.OuputLayer) {attributeGetTypes [#2]}
  n74 = Constant() {r=0.5}
  n75 = MaterialFunctionCall(Transition Phase=Saturate(None=Divide(A=@n71, B=@n72)), Height Texture=Subtract(A=@n73.Opacity, B=Multiply(A=@n56.R, B=@n74))) {materialFunction=&HeightLerp}
  n76 = LinearInterpolate(A=@n70.RGB, Alpha=@n75.Lerp Alpha No Contrast)
  n77 = ScalarParameter() {defaultValue=0.5, group=Scatter, parameterName=Seaweed Trans}
  n78 = Multiply(A=@n69.Opacity, B=ScalarParameter() {defaultValue=1, group=Scatter, parameterName=Seaweed HM})
  n79 = ScalarParameter() {defaultValue=0.5, group=Scatter, parameterName=Seaweed Con}
  n80 = SetMaterialAttributes!(
   MaterialAttributes=@n69.MaterialAttributes
   Base Color=Multiply(A=Multiply(A=@n69.Base Color, B=@n76), B=@n65)
   Opacity Mask=Multiply(
    A=MaterialFunctionCall(Transition Phase=@n77, Height Texture=@n78, Contrast=@n79) {materialFunction=&HeightLerp}.Alpha
    B=PerInstanceFadeAmount()
   )
   Pixel Depth Offset=@n68
   Specular=Multiply(A=@n69.Specular, B=LinearInterpolate(Alpha=@n75.Lerp Alpha No Contrast) {constA=0.2})
  ) {attributeSetTypes [
    #3
    #6
    #7
    #5
   ]}
  n81 = Multiply(
   A=MaterialFunctionCall(Transition Phase=ScalarParameter() {defaultValue=0.5, group=Scatter, parameterName=MaskValue}, Height Texture=@n73.Opacity) {materialFunction=&HeightLerp}.Lerp Alpha No Contrast
   B=PerInstanceFadeAmount()
  )
  n82 = SetMaterialAttributes!(
   MaterialAttributes
   Base Color=Multiply(A=GetMaterialAttributes(None=@n59.Blended Layers) {attributeGetTypes [#3]}.Base Color, B=@n70.RGB)
   Specular=Constant() {r=0.01}
   Roughness=Constant() {r=0.2}
  ) {attributeSetTypes [#3, #5, #4]}
  n83 = GetMaterialAttributes(None=StaticSwitchParameter(
    True=SetMaterialAttributes!(MaterialAttributes=@n58.OuputLayer, Pixel Depth Offset=@n68, Opacity Mask=@n81) {attributeSetTypes [#7, #6]}
    False=BlendMaterialAttributes(A=@n82, B=@n59.Blended Layers, Alpha=@n75.Lerp Alpha No Contrast)
   ) {group=Scatter, parameterName=ScatterMesh}) {attributeGetTypes [#3]}
  n84 = GetMaterialAttributes(None=SetMaterialAttributes!(MaterialAttributes=@n83.MaterialAttributes, Base Color=Multiply(A=@n83.Base Color, B=@n65)) {attributeSetTypes [#3]}) {attributeGetTypes [#3]}
  n85 = Constant3Vector() {constant=<rgba 0.0002 0.0833 0.1224 1>}
  n86 = Constant() {r=49.9663}
  n87 = Saturate(None=Divide(A=Add(A=@n63, B=@n86), B=Constant() {r=-171.32}))
  n88 = Constant() {r=193.8738}
  n89 = Saturate(None=Divide(A=Add(A=@n63, B=@n88), B=ScalarParameter() {defaultValue=-255, parameterName=WPO_Fade}))
  n90 = ScalarParameter() {defaultValue=15555, parameterName=ScaleEdge}
  n91 = AppendVector(A=@n90, B=@n90)
  n92 = Constant() {r=1}
  n93 = Add(A=VectorParameter() {parameterName=WPO, defaultValue=<rgba 0 0 -6000 1>}.RGB, B=Multiply(A=VertexNormalWS(), B=AppendVector(A=@n91, B=@n92)))
  n94 = StaticSwitchParameter(
   True=StaticSwitchParameter(
    True=SetMaterialAttributes!(MaterialAttributes=@n60.MaterialAttributes, Base Color=@n66, Opacity Mask=@n67, Pixel Depth Offset=@n68) {attributeSetTypes [#3, #6, #7]}
    False=StaticSwitchParameter(True=@n80, False=@n84.MaterialAttributes) {group=Scatter, parameterName=Seaweeds}
   ) {group=Scatter, parameterName=Grass}
   False=SetMaterialAttributes!(
    MaterialAttributes=@n84.MaterialAttributes
    World Position Offset=Multiply(A=@n89, B=@n93)
    Base Color=LinearInterpolate(A=@n84.Base Color, B=@n85, Alpha=@n87)
   ) {attributeSetTypes [#8, #3]}
  ) {parameterName=ScatterMesh}
  n95 = GetMaterialAttributes(None=@n94) {attributeGetTypes [#3, #1]}
  n96 = SetMaterialAttributes!(MaterialAttributes, Base Color=@n95.Base Color, Normal=@n95.Base Color) {attributeSetTypes [#3, #1]}
  ShadingPathSwitch(Default=@n94, Deferred=@n94, Forward=@n96, Mobile=@n96)
 ]
}
